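; A_MusicClockController is a Harmonix test Blueprint that displays the live state of a MetaSound music clock on screen.
; Each tick it reads tempo, time signature and the video-render-time bar and beat from its MusicClock variable.
; It prints them as formatted debug text lasting one frame's delta time.
; BeginPlay, overlap and key events (Up, Down, 1-6) are empty stubs, so nothing here controls the clock.
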(event EventBeginPlay)

(event Collision|EventActorBeginOverlap
  (params (OtherActor :type Actor%20Object%20Reference)))

(event Up
  (params (Key :type Key%20Structure)))

(event %31
  (params (Key :type Key%20Structure)))

(event %32
  (params (Key :type Key%20Structure)))

(event %33
  (params (Key :type Key%20Structure)))

(event %34
  (params (Key :type Key%20Structure)))

(event Down
  (params (Key :type Key%20Structure)))

(event %35
  (params (Key :type Key%20Structure)))

(event %36
  (params (Key :type Key%20Structure)))

(event EventTick
  (params (DeltaSeconds :type Float%20%28single-precision%29))
  (bind (_returnvalue_bar _returnvalue_beat) (Utilities|Struct|BreakMusicTimestamp (MusicClock|GetCurrentTimestamp (Variables|Default|GetMusicClock) "VideoRenderTime" :!class /Script/HarmonixMetasound.MusicClockComponent)))
  (bind (_outnumerator _outdenominator) (MusicClock|GetCurrentTimeSignature (Variables|Default|GetMusicClock) :!class /Script/HarmonixMetasound.MusicClockComponent))
  (Development|PrintString (Utilities|String|ToString(Text) (Utilities|Text|FormatText "NSLOCTEXT(\"[BC712BE7EB9E14E8240DD5F654B2368F]\", \"1A29E7B54B93FDD5DDAF32BA5DDFCFDB\", \"Tempo: {Tempo}\\r\\nTime Sig: {Num}/{Denom}\\r\\n{Bar}:{Beat}\")" (MusicClock|GetCurrentTempo (Variables|Default|GetMusicClock) :!class /Script/HarmonixMetasound.MusicClockComponent) _outnumerator _outdenominator _returnvalue_bar _returnvalue_beat)) true true "(R=0.000000,G=0.660000,B=1.000000,A=1.000000)" DeltaSeconds))

(fn ConstructionScript)
