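; Functional test actor that checks a Harmonix Music Clock driven by a Metasound metronome on an audio component.
; Setup connects and starts the clock; tick pushes time-scheduled tempo and the desired time signature as Metasound parameters, asserts the clock follows, and finishes after ten seconds.
; Bar and beat events check counters, sections, and smoothed and raw song positions against expected values.
; Several evaluation functions are empty stubs or never called.
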(event EventPrepareTest
  (bind _returnvalue (MusicClock|ConnecttoMetasoundonAudioComponent (Variables|Default|GetMusicClock) (Variables|Default|GetAudio)))
  (if _returnvalue
    (MusicClock|Start (Variables|Default|GetMusicClock))
    (else
      (FunctionalTesting|FinishTest "Failed" "Failed to connect Metasound on Audio Component to Music Clock"))))

(event EventStartTest
  (Default|AssertEqual(MusicClockState) (MusicClock|GetState (Variables|Default|GetMusicClock) :!class /Script/HarmonixMetasound.MusicClockComponent) "Running"))

(event EventTick
  (params (DeltaSeconds :type Float%20%28single-precision%29))
  (bind (_outnumerator _outdenominator) (MusicClock|GetCurrentTimeSignature (Variables|Default|GetMusicClock) :!class /Script/HarmonixMetasound.MusicClockComponent))
  (bind (_outnumerator_2 _outdenominator_2) (MusicClock|GetCurrentTimeSignature (Variables|Default|GetMusicClock) :!class /Script/HarmonixMetasound.MusicClockComponent))
  (bind (_outnumerator_3 _outdenominator_3) (MusicClock|GetCurrentTimeSignature (Variables|Default|GetMusicClock) :!class /Script/HarmonixMetasound.MusicClockComponent))
  (bind _musicclock (Variables|Default|GetMusicClock))
  (Utilities|FlowControl|Sequence
    (:then_0
      (bind _returnvalue (CallFunction|CalculateDesiredTempo (Variables|FunctionalTesting|GetTotalTime)))
      (Variables|Tempo&TimeSig|SetDesiredTempo _returnvalue)
      (Utilities|FlowControl|Sequence
        (:then_0
          (if (not (Math|Float|NearlyEqual(Float) (Variables|Tempo&TimeSig|GetDesiredTempo) (MusicClock|GetCurrentTempo (Variables|Default|GetMusicClock) :!class /Script/HarmonixMetasound.MusicClockComponent) 0.1))
            (Audio|Parameter|SetFloatParameter (Variables|Default|GetAudio) "Metronome.Tempo" (Variables|Tempo&TimeSig|GetDesiredTempo) :!class /Script/AudioExtensions.AudioParameterControllerInterface)
            (Utilities|FlowControl|Delay 0.1)
            (Asserts|AssertEqual(Float) (MusicClock|GetCurrentTempo (Variables|Default|GetMusicClock) :!class /Script/HarmonixMetasound.MusicClockComponent) (Variables|Tempo&TimeSig|GetDesiredTempo) "Music Clock Tempo" 0.01)
            (Asserts|AssertEqual(Float) (MusicClock|GetCurrentSecondsPerBeat _musicclock) (/ 6e+01 (Variables|Tempo&TimeSig|GetDesiredTempo)) "MusicClock->BeatDurationSec")
            (Asserts|AssertEqual(Float) (MusicClock|GetCurrentSecondsPerBar _musicclock) (* (/ 6e+01 (Variables|Tempo&TimeSig|GetDesiredTempo)) (* (Math|Conversions|ToFloat(Integer) _outnumerator) (/ 4.0 _outdenominator))) "MusicClock->BarDurationSec")))
        (:then_1
          (if (or (!= _outnumerator_3 (Variables|Tempo&TimeSig|GetDesiredTimeSigNum)) (!= _outdenominator_3 (Variables|Tempo&TimeSig|GetDesiredTimeSigDen)))
            (Audio|Parameter|SetParameters (Variables|Default|GetAudio) (Utilities|Array|MakeArray (Utilities|Struct|MakeAudioParameter :ParamName (Utilities|Name|MakeLiteralName "Metronome.TimeSig.Numerator") :FloatParam 0.0 :BoolParam false :IntParam (Variables|Tempo&TimeSig|GetDesiredTimeSigNum) :ParamType "Integer") (Utilities|Struct|MakeAudioParameter :ParamName (Utilities|Name|MakeLiteralName "Metronome.TimeSig.Denominator") :FloatParam 0.0 :BoolParam false :IntParam (Variables|Tempo&TimeSig|GetDesiredTimeSigDen) :ParamType "Integer")))
            (Utilities|FlowControl|Delay 0.1)
            (Asserts|AssertEqual(Integer) (MusicClock|GetCurrentTimeSignature (Variables|Default|GetMusicClock) :!class /Script/HarmonixMetasound.MusicClockComponent) (Variables|Tempo&TimeSig|GetDesiredTimeSigNum) "Music Clock Time Sig Numerator")
            (Asserts|AssertEqual(Integer) _outdenominator_2 (Variables|Tempo&TimeSig|GetDesiredTimeSigDen) "Music Clock Time Sig Denominator")))))
    (:then_2
      (if (>= (Variables|FunctionalTesting|GetTotalTime) 1e+01)
        (FunctionalTesting|FinishTest "Default")))))

(event BarEvent(MusicClock)
  (params (BarNumber :type Integer))
  (Asserts|AssertEqual(Integer) BarNumber (+ (Variables|Counts|GetBarCounter) 1) "Bar Number")
  (Math|Integer|IncrementInt (Variables|Counts|GetBarCounter)
    (:%20%20
      (Evaluation|EvaluateClockCurrentSection)
      (Evaluation|EvaluateClockSectionatMs (* (Variables|FunctionalTesting|GetTotalTime) 1e+03)))))

(event BeatEvent(MusicClock)
  (params (BeatNumber :type Integer) (BeatInBar :type Integer))
  (Asserts|AssertEqual(Integer) BeatNumber (Variables|Counts|GetBeatCounter) "Beat Number")
  (Evaluation|EvaluateClockSmoothedSongPos)
  (Evaluation|EvaluateClockRawSongPos)
  (Math|Integer|IncrementInt (Variables|Counts|GetBeatCounter)))

(event MusicClockConnectedEvent(MusicClock))

(event MusicClockDisconnectedEvent(MusicClock))

(fn AssertEqual(MusicClockState)
  (params (Actual :type EMusicClockState%20Enum :default "Stopped") (Expected :type EMusicClockState%20Enum :default "Stopped"))
  (bind _returnvalue (Asserts|AssertEqual(Integer) (Math|Conversions|ToInteger(Byte) Actual) (Math|Conversions|ToInteger(Byte) Expected) (Utilities|String|ToString(Text) (Utilities|Text|FormatText "INVTEXT(\"Clock State ({Actual} == {Expected})\")" (Utilities|String|EnumtoString Actual) (Utilities|String|EnumtoString Expected)))))
  (return _returnvalue))

(fn AssertMIDISongPosEqual
  (params (Actual :type MIDI%20Song%20Position%20Structure%20%28by%20ref%29) (Expected :type MIDI%20Song%20Position%20Structure%20%28by%20ref%29) (Error%20Tolerance%20%28Seconds%29 :type Float%20%28double-precision%29) (Error%20Tolerance%20%28Beats%29 :type Float%20%28double-precision%29) (Message :type String))
  (bind (_secondsfrombarone _secondsincludingcountin _timesignumerator _timesigdenominator _tempo _barsincludingcountin _beatsincludingcountin _beattype _timestamp) (Utilities|Struct|BreakMIDISongPosition Actual))
  (bind (_timestamp_bar _timestamp_beat) (Utilities|Struct|BreakMusicTimestamp _timestamp))
  (bind (_secondsfrombarone_2 _secondsincludingcountin_2 _timesignumerator_2 _timesigdenominator_2 _tempo_2 _barsincludingcountin_2 _beatsincludingcountin_2 _beattype_2 _timestamp_2) (Utilities|Struct|BreakMIDISongPosition Expected))
  (bind (_timestamp_bar_2 _timestamp_beat_2) (Utilities|Struct|BreakMusicTimestamp _timestamp_2))
  (bind _returnvalue (Asserts|AssertTrue (Math|Boolean|ANDBoolean (Math|Float|NearlyEqual(Float) _secondsfrombarone _secondsfrombarone_2 (Variables|GetError%20Tolerance%20%28Seconds%29)) (Math|Float|NearlyEqual(Float) _secondsincludingcountin _secondsincludingcountin_2 (Variables|GetError%20Tolerance%20%28Seconds%29)) (== _timesignumerator _timesignumerator_2) (== _timesigdenominator _timesigdenominator_2) (Math|Float|NearlyEqual(Float) _tempo _tempo_2 (Variables|GetError%20Tolerance%20%28Beats%29)) (Math|Float|NearlyEqual(Float) _barsincludingcountin _barsincludingcountin_2 (Variables|GetError%20Tolerance%20%28Beats%29)) (Math|Float|NearlyEqual(Float) _beatsincludingcountin _beatsincludingcountin_2 (Variables|GetError%20Tolerance%20%28Beats%29)) (== (Math|Conversions|ToInteger(Byte) _beattype) (Math|Conversions|ToInteger(Byte) _beattype_2)) (== _timestamp_bar _timestamp_bar_2) (Math|Float|NearlyEqual(Float) _timestamp_beat _timestamp_beat_2 (Variables|GetError%20Tolerance%20%28Beats%29))) Message))
  (if (not _returnvalue)
    (Development|PrintString (Utilities|String|ToString(Text) (Utilities|Text|FormatText "INVTEXT(\"Property: (Actual) vs (Expected)\\r\\nSeconds From Bar One: {A0} vs {E0}\\r\\nSeconds Including Count In: {A1} vs {E1}\\r\\nTime Sig Num: {A2} vs {E2}\\r\\nTime Sig Den: {A3} vs {E3}\\r\\nTempo: {A4} vs {E4}\\r\\nBars Including Count In: {A5} vs {E5}\\r\\nBeats Including Count In: {A6} vs {E6}\\r\\nBeat Type: {A7} vs {E7}\\r\\nTimestamp Bar: {A8} vs {E8}\\r\\nTimestamp Beat {A9} vs {E9}\")" _secondsfrombarone _secondsfrombarone_2 _secondsincludingcountin _secondsincludingcountin_2 _timesignumerator _timesignumerator_2 _timesigdenominator _timesigdenominator_2 _tempo _tempo_2 _barsincludingcountin _barsincludingcountin_2 _beatsincludingcountin _beatsincludingcountin_2 (Utilities|String|EnumtoString _beattype) (Utilities|String|EnumtoString _beattype_2) _timestamp_bar _timestamp_bar_2 _timestamp_beat _timestamp_beat_2))))
  (return _returnvalue))

(fn AssertMIDISongPosValuesEqual
  (params (MidiSongPos :type MIDI%20Song%20Position%20Structure%20%28by%20ref%29) (SecondsFromBarOne :type Float%20%28single-precision%29) (SecondsIncludingCountIn :type Float%20%28single-precision%29) (TimeSigNum :type Integer) (TimeSigDen :type Integer) (Tempo :type Float%20%28single-precision%29) (BarsIncludingCountIn :type Float%20%28single-precision%29) (BeatsIncludingCountIn :type Float%20%28single-precision%29) (BeatType :type EMusicalBeatType%20Enum :default "Downbeat") (Timestamp%20Bar :type Integer) (Timestamp%20Beat :type Float%20%28single-precision%29) (Error%20Tolerance%20%28Seconds%29 :type Float%20%28double-precision%29) (Error%20Tolerance%20%28Beats%29 :type Float%20%28double-precision%29) (What :type String))
  (bind (_secondsfrombarone _secondsincludingcountin _timesignumerator _timesigdenominator _tempo _barsincludingcountin _beatsincludingcountin _beattype _timestamp) (Utilities|Struct|BreakMIDISongPosition MidiSongPos))
  (bind (_timestamp_bar _timestamp_beat) (Utilities|Struct|BreakMusicTimestamp _timestamp))
  (bind _returnvalue (Asserts|AssertTrue (Math|Boolean|ANDBoolean (Math|Float|NearlyEqual(Float) _secondsfrombarone (Variables|GetSecondsFromBarOne) (Variables|GetError%20Tolerance%20%28Seconds%29)) (Math|Float|NearlyEqual(Float) _secondsincludingcountin (Variables|GetSecondsIncludingCountIn) (Variables|GetError%20Tolerance%20%28Seconds%29)) (== _timesignumerator (Variables|GetTimeSigNum)) (== _timesigdenominator (Variables|GetTimeSigDen)) (Math|Float|NearlyEqual(Float) _tempo (Variables|GetTempo) (Variables|GetError%20Tolerance%20%28Beats%29)) (Math|Float|NearlyEqual(Float) _barsincludingcountin (Variables|GetBarsIncludingCountIn) (Variables|GetError%20Tolerance%20%28Beats%29)) (Math|Float|NearlyEqual(Float) _beatsincludingcountin (Variables|GetBeatsIncludingCountIn) (Variables|GetError%20Tolerance%20%28Beats%29)) (== (Math|Conversions|ToInteger(Byte) _beattype) (Math|Conversions|ToInteger(Byte) (Variables|GetBeatType))) (== _timestamp_bar (Variables|GetTimestamp%20Bar)) (Math|Float|NearlyEqual(Float) _timestamp_beat (Variables|GetTimestamp%20Beat) (Variables|GetError%20Tolerance%20%28Beats%29))) (Variables|GetWhat)))
  (if (not _returnvalue)
    (Development|PrintString (Utilities|String|ToString(Text) (Utilities|Text|FormatText "INVTEXT(\"Property: (Actual) vs (Expected)\\r\\nSeconds From Bar One: {A0} vs {E0}\\r\\nSeconds Including Count In: {A1} vs {E1}\\r\\nTime Sig Num: {A2} vs {E2}\\r\\nTime Sig Den: {A3} vs {E3}\\r\\nTempo: {A4} vs {E4}\\r\\nBars Including Count In: {A5} vs {E5}\\r\\nBeats Including Count In: {A6} vs {E6}\\r\\nBeat Type: {A7} vs {E7}\\r\\nTimestamp Bar: {A8} vs {E8}\\r\\nTimestamp Beat {A9} vs {E9}\")" _secondsfrombarone SecondsFromBarOne _secondsincludingcountin SecondsIncludingCountIn _timesignumerator TimeSigNum _timesigdenominator TimeSigDen _tempo Tempo _barsincludingcountin BarsIncludingCountIn _beatsincludingcountin BeatsIncludingCountIn (Utilities|String|EnumtoString _beattype) (Utilities|String|EnumtoString BeatType) _timestamp_bar Timestamp%20Bar _timestamp_beat Timestamp%20Beat))))
  (return _returnvalue))

(fn CalculateDesiredTempo
  (params (TimeSeconds :type Float%20%28double-precision%29))
  (bind _returnvalue (Math|Float|Floor TimeSeconds))
  (bind _timetotempo (LocalVariables|GetTimeToTempo))
  (bind (_value _returnvalue_2) (Utilities|Map|Find _timetotempo _returnvalue))
  (if _returnvalue_2
    (return _value)
    (else
      (return 1.2e+02))))

(fn CalculateDesiredTimeSig
  (params (TimeSeconds :type Float%20%28double-precision%29))
  (bind _returnvalue (Math|Float|Truncate TimeSeconds))
  (bind _timesigpoints (LocalVariables|GetTimeSigPoints))
  (bind (_value _returnvalue_2) (Utilities|Map|Find _timesigpoints _returnvalue))
  (bind (_x _y) (Math|Vector2D|BreakVector2D _value))
  (if _returnvalue_2
    (return (Math|Float|Floor _x) (Math|Float|Floor _y))
    (else
      (return 4 4))))

(fn DisplayClockValues
  (params (Duration :type Float%20%28single-precision%29))
  (bind (_outnumerator _outdenominator) (MusicClock|GetCurrentTimeSignature (Variables|Default|GetMusicClock) :!class /Script/HarmonixMetasound.MusicClockComponent))
  (Development|PrintString (Utilities|String|ToString(Text) (Utilities|Text|FormatText "INVTEXT(\"{Num}/{Den}\\r\\n\\r\\n{BarDur}:{BeatDur}\\r\\n\\r\\n{AdvanceRate}\")" _outnumerator _outdenominator (MusicClock|GetCurrentSecondsPerBar (Variables|Default|GetMusicClock)) (MusicClock|GetCurrentSecondsPerBeat (Variables|Default|GetMusicClock)) (MusicClock|GetCurrentClockAdvanceRate (Variables|Default|GetMusicClock) :!class /Script/HarmonixMetasound.MusicClockComponent))) true true "(R=0.000000,G=0.660000,B=1.000000,A=1.000000)" Duration))

(fn EvaluateClockAtMs)

(fn EvaluateClockCurrentSection
  (bind _output_name_3_c002cfce4c4002489e1d7e8a78d797a4 (Utilities|Struct|BreakFTMusicClockSectionData (Utilities|Array|Get(aref) (Variables|Sections|GetExpectedSections) (Variables|Sections|GetExpectedSectionIndex))))
  (bind (_output_name_3_c002cfce4c4002489e1d7e8a78d797a4_2 _output_startms_7_29e8aef94ace39fbbb2ff288b65e3e5d _output_lengthms_9_ec2dc68744e812c4c76b03a66058b6e7) (Utilities|Struct|BreakFTMusicClockSectionData (Utilities|Array|Get(aref) (Variables|Sections|GetExpectedSections) (Variables|Sections|GetExpectedSectionIndex))))
  (bind (_output_name_3_c002cfce4c4002489e1d7e8a78d797a4_3 _output_startms_7_29e8aef94ace39fbbb2ff288b65e3e5d_2) (Utilities|Struct|BreakFTMusicClockSectionData (Utilities|Array|Get(aref) (Variables|Sections|GetExpectedSections) (Variables|Sections|GetExpectedSectionIndex))))
  (Asserts|AssertEqual(Integer) (Section|GetNumSections (Variables|Default|GetMusicClock)) 0 "MusicClock->GetNumSections()")
  (Asserts|AssertEqual(Integer) (Utilities|Array|Length (MusicClock|GetSongSections (Variables|Default|GetMusicClock))) (Utilities|Array|Length (Variables|Sections|GetExpectedSections)) "MusicClock->GetSongSections().Length()")
  (if (> (Section|GetNumSections (Variables|Default|GetMusicClock)) 0)
    (bind _returnvalue (Asserts|AssertEqual(Integer) (MusicClock|GetCurrentSectionIndex (Variables|Default|GetMusicClock) "VideoRenderTime" :!class /Script/HarmonixMetasound.MusicClockComponent) (Variables|Sections|GetExpectedSectionIndex) "MusicClock->CurrentSectionIndex()"))
    (if _returnvalue
      (Asserts|AssertEqual(String) (MusicClock|GetCurrentSectionName (Variables|Default|GetMusicClock) "VideoRenderTime" :!class /Script/HarmonixMetasound.MusicClockComponent) _output_name_3_c002cfce4c4002489e1d7e8a78d797a4)
      (Asserts|AssertEqual(Float) (MusicClock|GetCurrentSectionStartMs (Variables|Default|GetMusicClock) "VideoRenderTime" :!class /Script/HarmonixMetasound.MusicClockComponent) _output_startms_7_29e8aef94ace39fbbb2ff288b65e3e5d_2 "MusicClock->CurrentSectionStartMs()" 0.1)
      (Asserts|AssertEqual(Float) (MusicClock|GetCurrentSectionLengthMs (Variables|Default|GetMusicClock) "VideoRenderTime" :!class /Script/HarmonixMetasound.MusicClockComponent) _output_lengthms_9_ec2dc68744e812c4c76b03a66058b6e7 "MusicClock->CurrentSectionLengthMs()" 0.1))
    (else
      (Asserts|AssertEqual(Integer) (MusicClock|GetCurrentSectionIndex (Variables|Default|GetMusicClock) "VideoRenderTime" :!class /Script/HarmonixMetasound.MusicClockComponent) -1 "MusicClock->CurrentSectionIndex()"))))

(fn EvaluateClockDeltas)

(fn EvaluateClockLoopSeekFlags)

(fn EvaluateClockPlayerSongPos
  (bind (_secondsfrombarone _secondsincludingcountin _timesignum _timesigden _tempo _barsincludingcountin _beatsincludingcountin _beattype _timestampbar _timestampbeat) (CallFunction|GetExpectedSongPosValues))
  (bind _returnvalue (MusicClock|GetCurrentPlayerExperiencedSongPos (Variables|Default|GetMusicClock)))
  (bind _musicclock (Variables|Default|GetMusicClock))
  (Default|AssertMIDISongPosEqual (MusicClock|GetCurrentPlayerExperiencedSongPos _musicclock) _returnvalue 0.0 0.0 "Getter vs Member: PlayerSongPos")
  (Default|AssertMIDISongPosValuesEqual _returnvalue _secondsfrombarone _secondsincludingcountin _timesignum _timesigden _tempo _barsincludingcountin _beatsincludingcountin _beattype _timestampbar _timestampbeat 1.0 0.1 "MusicClock->PlayerSongPos vs Beat Counter"))

(fn EvaluateClockRawSongPos
  (bind _returnvalue (MusicClock|GetCurrentRawAudioRenderSongPos (Variables|Default|GetMusicClock)))
  (bind (_secondsfrombarone _secondsincludingcountin) (Utilities|Struct|BreakMIDISongPosition _returnvalue))
  (bind _musicclock (Variables|Default|GetMusicClock))
  (bind _totaltime (Variables|FunctionalTesting|GetTotalTime))
  (Default|AssertMIDISongPosEqual (MusicClock|GetCurrentRawAudioRenderSongPos _musicclock) _returnvalue 1.0 0.1 "Getter vs Member: RawSongPos")
  (Asserts|AssertEqual(Float) _secondsfrombarone _totaltime "MusicClock->RawSongPos.SecondsFromBarOne" 1.0)
  (Asserts|AssertEqual(Float) _secondsincludingcountin _totaltime "MusicClock->RawSongPos.SecondsIncludingCountIn" 1.0))

(fn EvaluateClockSectionAtMs
  (params (Ms :type Float%20%28single-precision%29))
  (bind _sectionindex (CallFunction|FindExpectedSectionatMs Ms))
  (bind _expectedsections (Variables|Sections|GetExpectedSections))
  (bind (_output_name_3_c002cfce4c4002489e1d7e8a78d797a4 _output_startms_7_29e8aef94ace39fbbb2ff288b65e3e5d _output_lengthms_9_ec2dc68744e812c4c76b03a66058b6e7) (Utilities|Struct|BreakFTMusicClockSectionData (Utilities|Array|Get(aref) _expectedsections _sectionindex)))
  (bind _returnvalue (Asserts|AssertEqual(Integer) (Section|GetNumSections (Variables|Default|GetMusicClock)) (Utilities|Array|Length (Variables|Sections|GetExpectedSections)) "MusicClock->GetNumSections()"))
  (if _returnvalue
    (if (> (Section|GetNumSections (Variables|Default|GetMusicClock)) 0)
      (if (>= _sectionindex 0)
        (Asserts|AssertEqual(String) (Section|GetSectionNameatMs (Variables|Default|GetMusicClock) Ms) _output_name_3_c002cfce4c4002489e1d7e8a78d797a4 (Utilities|String|ToString(Text) (Utilities|Text|FormatText "INVTEXT(\"MusicClock->GetSectionNameAtMs({Ms}).Name\")" Ms)))
        (Asserts|AssertEqual(Float) (Section|GetSectionLengthMsatMs (Variables|Default|GetMusicClock) Ms) _output_lengthms_9_ec2dc68744e812c4c76b03a66058b6e7 (Utilities|String|ToString(Text) (Utilities|Text|FormatText "INVTEXT(\"MusicClock->GetSectionNameAtMs({Ms}).Length\")" Ms)) 0.1)
        (Asserts|AssertEqual(Float) (Section|GetSectionStartMsatMs (Variables|Default|GetMusicClock) Ms) _output_startms_7_29e8aef94ace39fbbb2ff288b65e3e5d (Utilities|String|ToString(Text) (Utilities|Text|FormatText "INVTEXT(\"MusicClock->GetSectionNameAtMs({Ms}).Start\")" Ms)) 0.1)
        (Asserts|AssertEqual(Float) (Section|GetSectionEndMsatMs (Variables|Default|GetMusicClock) Ms) (+ _output_startms_7_29e8aef94ace39fbbb2ff288b65e3e5d _output_lengthms_9_ec2dc68744e812c4c76b03a66058b6e7) (Utilities|String|ToString(Text) (Utilities|Text|FormatText "INVTEXT(\"MusicClock->GetSectionNameAtMs({Ms}).End\")" Ms)) 0.1)
        (else
          _))
      (else
        _))
    (Asserts|AssertEqual(String) (Section|GetSectionNameatMs (Variables|Default|GetMusicClock) Ms) "" (Utilities|String|ToString(Text) (Utilities|Text|FormatText "INVTEXT(\"Expected No Section: MusicClock->GetSectionNameAtMs({Ms})\")" Ms)))))

(fn EvaluateClockSmoothedSongPos
  (bind (_secondsfrombarone _secondsincludingcountin _timesignum _timesigden _tempo _barsincludingcountin _beatsincludingcountin _beattype _timestampbar _timestampbeat) (CallFunction|GetExpectedSongPosValues))
  (bind _returnvalue (MusicClock|GetCurrentSmoothedAudioRenderSongPos (Variables|Default|GetMusicClock)))
  (bind _musicclock (Variables|Default|GetMusicClock))
  (Default|AssertMIDISongPosEqual (MusicClock|GetCurrentSmoothedAudioRenderSongPos _musicclock) _returnvalue 0.0 0.0 "Getter vs Member: SmoothedAudioRenderSongPos")
  (Default|AssertMIDISongPosValuesEqual _returnvalue _secondsfrombarone _secondsincludingcountin _timesignum _timesigden _tempo _barsincludingcountin _beatsincludingcountin _beattype _timestampbar _timestampbeat 1.0 0.1 "MusicClock->SmoothedPos vs Beat Counter"))

(fn EvaluateClockSongMaps)

(fn EvaluateClockSongPos)

(fn EvaluateClockTimeConversions)

(fn EvaluateClockVideoSongPos
  (bind (_secondsfrombarone _secondsincludingcountin _timesignum _timesigden _tempo _barsincludingcountin _beatsincludingcountin _beattype _timestampbar _timestampbeat) (CallFunction|GetExpectedSongPosValues))
  (bind _returnvalue (MusicClock|GetCurrentVideoRenderSongPos (Variables|Default|GetMusicClock)))
  (bind _musicclock (Variables|Default|GetMusicClock))
  (Default|AssertMIDISongPosEqual (MusicClock|GetCurrentVideoRenderSongPos _musicclock) _returnvalue 0.0 0.0 "Getter vs Member: VideoSongPos")
  (Default|AssertMIDISongPosValuesEqual _returnvalue _secondsfrombarone _secondsincludingcountin _timesignum _timesigden _tempo _barsincludingcountin _beatsincludingcountin _beattype _timestampbar _timestampbeat 1.0 0.1 "MusicClock->VideoSongPos vs Beat Counter"))

(fn FindExpectedSectionAtMs
  (params (Ms :type Float%20%28double-precision%29))
  (bind _expectedsections (Variables|Sections|GetExpectedSections))
  (if (Utilities|Array|IsNotEmpty _expectedsections)
    (for (_array_element _array_index) _expectedsections
      (bind (_array_element_name_3_c002cfce4c4002489e1d7e8a78d797a4 _array_element_startms_7_29e8aef94ace39fbbb2ff288b65e3e5d _array_element_lengthms_9_ec2dc68744e812c4c76b03a66058b6e7) (Utilities|Struct|BreakFTMusicClockSectionData _array_element))
      (if (and (>= _array_element_startms_7_29e8aef94ace39fbbb2ff288b65e3e5d Ms) (< Ms (+ _array_element_startms_7_29e8aef94ace39fbbb2ff288b65e3e5d _array_element_lengthms_9_ec2dc68744e812c4c76b03a66058b6e7)))
        (return _array_index)))
    (else
      (return -1))))

(fn GetExpectedSongPosValues
  (return (Variables|FunctionalTesting|GetTotalTime) (Variables|FunctionalTesting|GetTotalTime) (Variables|Tempo&TimeSig|GetDesiredTimeSigNum) (Variables|Tempo&TimeSig|GetDesiredTimeSigDen) (Variables|Tempo&TimeSig|GetDesiredTempo) (/ (Math|Conversions|ToFloat(Integer) (Variables|Counts|GetBeatCounter)) 4.0) (Math|Conversions|ToFloat(Integer) (Variables|Counts|GetBeatCounter)) "Normal" (+ (/ (Variables|Counts|GetBeatCounter) 4) 1) (Math|Conversions|ToFloat(Integer) (+ (Math|Integer|%25%28Integer%29 (Variables|Counts|GetBeatCounter) 4) 1))))

(fn ConstructionScript)
