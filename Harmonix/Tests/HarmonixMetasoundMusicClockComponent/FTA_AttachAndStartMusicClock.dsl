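; Functional test that checks a Music Clock component can attach to a MetaSound on the test object's audio component.
; On start it finds the Audio component, adds a MusicClockComponent at runtime and connects the clock to it.
; A successful connection starts the clock and passes without checking playback; a missing audio component or failed connection fails the test.
; The prepare event is empty.
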(event FunctionalTesting|EventOnStart
  (params (Test :type Functional%20Test%20Object%20Reference))
  (bind _returnvalue (Actor|GetComponentbyClass Test "/Script/Engine.AudioComponent"))
  (if (Utilities|IsValid _returnvalue :!class /Script/Engine.KismetSystemLibrary)
    (bind _returnvalue_2 (Game|AddComponentbyClass Test 0 "/Script/HarmonixMetasound.MusicClockComponent"))
    (bind _returnvalue_3 (MusicClock|ConnecttoMetasoundonAudioComponent _returnvalue_2 _returnvalue))
    (if _returnvalue_3
      (MusicClock|Start _returnvalue_2)
      (FunctionalTesting|Finish true)
      (else
        (FunctionalTesting|Finish false)))
    (else
      (FunctionalTesting|Finish false))))

(event FunctionalTesting|EventPrepare
  (params (Test :type Functional%20Test%20Object%20Reference)))
