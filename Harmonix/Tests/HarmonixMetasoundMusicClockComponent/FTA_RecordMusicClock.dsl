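; Functional test that runs a Harmonix MetaSound music clock for a set record time while displaying its position.
; Prepare caches the MusicClockComponent found on the test actor, and start fails the test if that component is invalid.
; Each tick prints the clock's video-render-time bar:beat to screen and log, accumulating delta time until the record time is reached, then finishes as passed without checking timestamp values.
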(event FunctionalTesting|EventPrepare
  (params (Test :type Functional%20Test%20Object%20Reference))
  (bind _output_get (Variables|Default|SetMusicClock (Actor|GetComponentbyClass Test "/Script/HarmonixMetasound.MusicClockComponent")))
  (Utilities|IsValid _output_get))

(event FunctionalTesting|EventOnStart
  (params (Test :type Functional%20Test%20Object%20Reference))
  (Variables|Default|GetMusicClock
    (:then)
    (:else
      (FunctionalTesting|Finish false))))

(event FunctionalTesting|EventTick
  (params (Test :type Functional%20Test%20Object%20Reference) (DeltaSeconds :type Float%20%28single-precision%29))
  (bind (_returnvalue_bar _returnvalue_beat) (Utilities|Struct|BreakMusicTimestamp (MusicClock|GetCurrentTimestamp (Variables|Default|GetMusicClock) "VideoRenderTime" :!class /Script/HarmonixMetasound.MusicClockComponent)))
  (bind _output_get (Variables|Default|SetTotalTime (+ (Variables|Default|GetTotalTime) DeltaSeconds)))
  (if (>= _output_get (Variables|Default|GetRecordTime))
    (FunctionalTesting|Finish true)
    (else
      (Development|PrintString 0 (Utilities|String|ToString(Text) (Utilities|Text|FormatText "NSLOCTEXT(\"[F52121DF6444FE63AB54986215441759]\", \"BF876E754AFB03099B401CAB01127F1E\", \"{Bar}:{Beat}\")" _returnvalue_bar _returnvalue_beat)) true true "(R=0.000000,G=0.660000,B=1.000000,A=1.000000)" DeltaSeconds))))
