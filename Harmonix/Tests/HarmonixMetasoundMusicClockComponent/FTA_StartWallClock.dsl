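; Functional test for the Harmonix MetaSound Music Clock component, exercising a clock driven by wall-clock time.
; On test start it adds the component, sets its default tempo from the Tempo variable and connects it to the wall clock using the TempoMap variable.
; It then binds beat and bar events that print their numbers, starts the clock and immediately finishes as passed, without asserting on clock output.
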(event FunctionalTesting|EventOnStart
  (params (Test :type Functional%20Test%20Object%20Reference))
  (bind _returnvalue (Game|AddComponentbyClass Test 0 "/Script/HarmonixMetasound.MusicClockComponent"))
  (Variables|/Script/HarmonixMetasound.MusicClockComponent|SetDefaultTempo (Variables|Default|GetTempo) _returnvalue)
  (MusicClock|ConnecttoWallClockforMidi _returnvalue (Variables|Default|GetTempoMap))
  (Utilities|FlowControl|Sequence
    (:then_0
      (MusicClock|BindEventtoBeatEvent _returnvalue (delegate Custom|OnBeat))
      (MusicClock|AssignBarEvent _returnvalue (delegate Custom|OnBar)))
    (:then_1
      (MusicClock|Start _returnvalue)
      (FunctionalTesting|Finish true))))

(event Custom|OnBeat
  (params (BeatNumber :type Integer) (BeatInBar :type Integer))
  (Development|PrintString 0 (Utilities|String|ToString(Text) (Utilities|Text|FormatText "NSLOCTEXT(\"[37C8C3A10293C868B726C164F46B210F]\", \"624012A34F9BDA53AA550190B09A0C3D\", \"Beat: {Beat}\\r\\nBeat In Bar: {BeatInBar}\")" BeatNumber BeatInBar))))

(event Custom|OnBar
  (params (BarNumber :type Integer))
  (Development|PrintString 0 (Utilities|String|ToString(Text) (Utilities|Text|FormatText "NSLOCTEXT(\"[37C8C3A10293C868B726C164F46B210F]\", \"B8DA182146E600AE5235BBB59F8F7382\", \"Bar: {Bar}\")" BarNumber))))
