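# Mono MetaSound source in the Harmonix music clock component tests that generates a MIDI clock with an audible metronome click.
# A trigger-to-transport node starts the metronome on play, with tempo, speed and time signature exposed as inputs.
# The clock feeds two subdivision triggers on different grids, each firing an AD envelope over a differently pitched sine.
# The two clicks are mixed to the audio output, and the clock is exposed as an output.
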MetaSoundSource /Harmonix/Tests/HarmonixMetasoundMusicClockComponent/MSS_MetronomeClockGenerator
inputs [
 Metronome.Speed {type=Float}
 Metronome.Tempo {type=Float}
 Metronome.TimeSig.Denominator {type=Int32}
 Metronome.TimeSig.Numerator {type=Int32}
 UE.Source.OnPlay {type=Trigger}
]
interfaces [UE.OutputFormat.Mono:1.0, UE.Source:1.1]
outputs [
 AudioOut {type=Audio}
 MIDI Clock {type=MIDIClock}
 "UE.OutputFormat.Mono.Audio:0" {type=Audio}
]
pages [Default {
  input_defaults {Metronome.Speed=1, Metronome.Tempo=110, Metronome.TimeSig.Denominator=8, Metronome.TimeSig.Numerator=6, UE.Source.OnPlay=null}
  nodes [
   HarmonixNodesMetronome_0 = HarmonixNodes.Metronome(
    Speed=^Metronome.Speed
    Tempo=^Metronome.Tempo
    Time Sig. Denominator=^Metronome.TimeSig.Denominator
    Time Sig. Numerator=^Metronome.TimeSig.Numerator
    Transport=HarmonixNodes.TriggerToTransport(Play=^UE.Source.OnPlay).Transport
   )
   ADEnvelopeADEnvelope_0 = AD Envelope.AD Envelope:Audio(Trigger=HarmonixNodes.MidiClockSubdivisionTrigger(MIDI Clock=@HarmonixNodesMetronome_0.MIDI Clock).Trigger Out) {also {defaults {Decay Time=0.07}}}
   MIDIToFrequency_1 = UE.MIDI To Frequency:Float() {also {defaults {MIDI In=72}}}
   ADEnvelopeADEnvelope_1 = AD Envelope.AD Envelope:Audio(Trigger=HarmonixNodes.MidiClockSubdivisionTrigger(MIDI Clock=@HarmonixNodesMetronome_0.MIDI Clock) {also {defaults {Base Grid Size=0}}}.Trigger Out) {also {defaults {Decay Time=0.1}}}
   MIDIToFrequency_0 = UE.MIDI To Frequency:Float()
   AudioMixerAudioMixerMono2_0 = "AudioMixer.Audio Mixer (Mono, 2)"(
    In 0=UE.Multiply:Audio(AdditionalOperands=@ADEnvelopeADEnvelope_0.Out Envelope, PrimaryOperand=UE.Sine:Audio(Frequency=@MIDIToFrequency_1.Out Frequency).Audio).Out
    In 1=UE.Multiply:Audio(AdditionalOperands=@ADEnvelopeADEnvelope_1.Out Envelope, PrimaryOperand=UE.Sine:Audio(Frequency=@MIDIToFrequency_0.Out Frequency).Audio).Out
   )
  ]
  links [
   {dst="@AudioOut", src=AudioMixerAudioMixerMono2_0.Out}
   {dst="@UE.OutputFormat.Mono.Audio:0", src=AudioMixerAudioMixerMono2_0.Out}
   {dst="@MIDI Clock", src=HarmonixNodesMetronome_0.MIDI Clock}
  ]
 }]
properties {
 assetImportData=&self:AssetImportData
 bLooping=true
 duration=10000
 sampleRate=48000
}
